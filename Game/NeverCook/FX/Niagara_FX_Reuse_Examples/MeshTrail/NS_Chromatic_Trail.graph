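# Niagara system for a chromatic mesh trail, filed under FX reuse examples, with a single AuraLeft emitter.
# A burst spawns mesh particles that are not killed when their lifetime elapses.
# Update scales size and color, writes dynamic material parameters, and picks camera offset from a float array indexed by particle partition, layering copies in depth.
# An integer User.RankedProgression is exposed, but none of the shown modules read it.
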NiagaraSystem /Game/NeverCook/FX/Niagara_FX_Reuse_Examples/MeshTrail/NS_Chromatic_Trail
dependencies {usedDataInterfaces [&NiagaraDataInterfaceArrayFloat], usedDynamicInputs [], usedModules [&SystemState]}
emitters {AuraLeft {
  emitterUpdateScript [
   &EmitterState(
    Inactive Response=::NewEnumerator1
    Life Cycle Mode=::NewEnumerator1
    Loop Behavior=::NewEnumerator1
    Loop Duration Mode=::NewEnumerator1
    UseLoopDelay=false
   )
   &SpawnBurst_Instantaneous(Use Spawn Probability=false, UseLoopCountLimit=false)
  ]
  particleSpawnScript [&InitializeParticle(
    Lifetime=2
    Mesh Scale=<1 1 0.5>
    Mesh Scale Mode=::NewEnumerator4
    Sprite Size Mode=::NewEnumerator3
    Uniform Sprite Size=50
    UsePositionOffset=false
    Write Material Random=true
   )]
  particleUpdateScript [
   &CameraOffset(Camera Offset Amount=&SelectFloatFromArray(Direct Array Index=$Particles.PartitionParticles.Partition, Float Selection Array=DataInterface() {propertyValues {FloatData [-10,0,0]}}))
   &ParticleState(Kill Particles When Lifetime Has Elapsed=false, Let Infinitely Lived Particles Die When Emitter Deactivates=true)
   &ScaleMeshSize(Scale Factor=<1 1 3>)
   &ScaleColor(
    Color Value To Scale=$Particles.Color
    Scale RGB=&Multiply_VectorByFloat(Vector=<1 1 1>, Float=1)
    ScaleA=true
    ScaleRGB=true
   )
   &DynamicMaterialParameters(
    Index 0 Param 1=1
    Module.Index0Param0Write=true
    Module.Index0Param1Write=true
    Module.Index0Param2Write=true
    Module.Index0Param3Write=true
   )
  ]
  rendererClasses [&NiagaraMeshRendererProperties]
  renderers [{rendererClass=&NiagaraMeshRendererProperties, rendererIndex=0}]
 }}
summary {
 systemName=NS_Chromatic_Trail
 emitters [{emitterName=AuraLeft, rendererClasses [&NiagaraMeshRendererProperties]}]
 userVariables [User.RankedProgression {type=NiagaraInt32}]
}
