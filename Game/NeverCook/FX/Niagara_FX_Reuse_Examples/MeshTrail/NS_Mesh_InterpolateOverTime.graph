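# Niagara example system from the MeshTrail FX reuse folder that computes a lagging, smoothed position for a mesh particle, apparently for a trail effect.
# A single emitter burst-spawns particles that are not killed when their lifetime elapses and draws them with a mesh renderer.
# Each update, InterpolateOverTime eases toward Particles.Position, and its moving average is packed with alpha 1 into the Vector4 InterpolateDy, presumably for the mesh material.
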NiagaraSystem /Game/NeverCook/FX/Niagara_FX_Reuse_Examples/MeshTrail/NS_Mesh_InterpolateOverTime
dependencies {usedDataInterfaces [], usedDynamicInputs [], usedModules [&SystemState]}
emitters {Empty {
  emitterUpdateScript [
   &EmitterState(
    Inactive Response=::NewEnumerator1
    Life Cycle Mode=::NewEnumerator1
    Loop Behavior=::NewEnumerator1
    Loop Duration Mode=::NewEnumerator1
    UseLoopDelay=false
   )
   &SpawnBurst_Instantaneous(Use Spawn Probability=false, UseLoopCountLimit=false)
  ]
  particleSpawnScript [&InitializeParticle(Color Mode=::NewEnumerator0, Mesh Scale Mode=::NewEnumerator4, UsePositionOffset=false, Write Material Random=true)]
  particleUpdateScript [
   &ParticleState(Kill Particles When Lifetime Has Elapsed=false)
   /SetVariables() {isSetParameters=true}
   &InterpolateOverTime(Rate Of Change=5, Target Vector Value:NiagaraPosition=$Particles.Position, Variable Type=::NewEnumerator2)
   /SetVariables(Particles.InterpolateDy:Vector4f=&MakeVector4FromVectorAndFloat("Vector (RGB)"=$Particles.InterpolateOverTime.InterpolateOverTime_V3_.Moving Average, "Float (Alpha)"=1)) {isSetParameters=true}
  ]
  rendererClasses [&NiagaraMeshRendererProperties]
  renderers [{rendererClass=&NiagaraMeshRendererProperties, rendererIndex=0}]
 }}
summary {
 systemName=NS_Mesh_InterpolateOverTime
 emitters [{emitterName=Empty, rendererClasses [&NiagaraMeshRendererProperties]}]
 userVariables []
}
