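# Niagara mesh-trail effect, from a folder of FX reuse examples, that by its name draws a flame trail for a punch.
# Its single emitter burst-spawns short-lived HDR orange mesh particles, tilted slightly by two stacked orientation modules.
# Over each particle's life, curves shrink the mesh along X, fade alpha out and ramp a dynamic material parameter to animate the material.
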NiagaraSystem /Game/NeverCook/FX/Niagara_FX_Reuse_Examples/MeshTrail/NS_PunchFlameTrail
dependencies {usedDataInterfaces [&NiagaraDataInterfaceCurve], usedDynamicInputs [], usedModules [&SystemState]}
emitters {PunchTrailFlame001 {
  emitterUpdateScript [
   &EmitterState(Life Cycle Mode=::NewEnumerator1, Loop Behavior=::NewEnumerator1, Loop Duration=0.5, UseLoopDelay=false)
   &SpawnBurst_Instantaneous(Use Spawn Probability=false, UseLoopCountLimit=false)
  ]
  particleSpawnScript [
   &InitializeParticle(Color=<rgba 10 2.2111 0 1>, Lifetime=0.5, Mesh Scale Mode=::NewEnumerator4, UsePositionOffset=true, Write Material Random=true)
   &InitialMeshOrientation(Mesh Orientation Mode=::NewEnumerator2, Rotation=<0 0 -0.02>, Use Rotation Vector=true)
   &InitialMeshOrientation(Mesh Orientation Mode=::NewEnumerator2, Rotation=<0 -0.01 0>, Use Rotation Vector=true) {module=InitialMeshOrientation001}
  ]
  particleUpdateScript [
   &ParticleState()
   &ScaleMeshSize(Scale Factor=&MakeVector(
     X=&FloatFromCurve(FloatCurve=DataInterface() {propertyValues {Curve {keys [
          {arriveTangent=-3.5245, interpMode=RCIM_Cubic, leaveTangent=-3.5245, tangentMode=RCTM_User, time=0, value=3}
          {arriveTangent=-0.278, tangentMode=RCTM_User, time=1, value=1.5}
         ]}}})
     Y=1
     Z=1
    ))
   &ScaleColor(
    Scale Alpha=&FloatFromCurve(FloatCurve=DataInterface() {propertyValues {Curve {keys [<at 0 1>,<at 1 0>]}}})
    ScaleA=true
    ScaleRGB=true
   )
   &DynamicMaterialParameters(
    Index 0 Param 1=&FloatFromCurve(FloatCurve=DataInterface() {propertyValues {Curve {keys [<at 0 0>,<at 1 0.5>]}}})
    Module.Index0Param0Write=true
    Module.Index0Param1Write=true
    Module.Index0Param2Write=true
    Module.Index0Param3Write=true
   )
  ]
  rendererClasses [&NiagaraMeshRendererProperties]
  renderers [{rendererClass=&NiagaraMeshRendererProperties, rendererIndex=0}]
 }}
summary {
 systemName=NS_PunchFlameTrail
 emitters [{emitterName=PunchTrailFlame001, rendererClasses [&NiagaraMeshRendererProperties]}]
 userVariables []
}
